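# Looping aura effect for the Canary pistol's Drain FX, built from two Niagara emitters.
# Flow steadily spawns tilted meshes that grow, fade from blue through purple to dark red, and animate a randomly offset material parameter under drag and a ramping negative point force at a hard-coded world position.
# GlowFlare bursts a particle that lasts until deactivation, driving layered sprites with per-frame random flicker and a cyan light.
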NiagaraSystem /PistolCanary/Assets/FX/Drain/NS_Pistol_Canary_MM_Aura_Loop
dependencies {
 usedDataInterfaces [&NiagaraDataInterfaceCurve, &NiagaraDataInterfaceColorCurve]
 usedDynamicInputs []
 usedModules [&SystemState]
}
emitters {
 Flow {
  emitterUpdateScript [
   &EmitterState()
   &SpawnRate(SpawnRate=2, Use Spawn Probability=false)
  ]
  particleSpawnScript [
   &InitializeParticle(
    Lifetime=&RandomRangeFloat(Minimum=2, Maximum=4)
    Mesh Scale Mode=::NewEnumerator3
    Mesh Uniform Scale=&MakeCustomFloatFromBool(
     Bool=&SetBoolByIntComparison(A=&ReturnExecIndex(), B=0, Comparison Type=::NewEnumerator4)
     True Float=0.8
     False Float=1.1
    )
    UsePositionOffset=false
    Write Material Random=true
   )
   &InitialMeshOrientation(Orientation Axis=<0 0 1>, Orientation Vector=<0 0 1>, Rotation=<0 0.25 0>, Use Orientation Vector=true, Use Rotation Vector=true) {module=InitialMeshOrientation001}
   &InitialMeshOrientation(Rotation=&RandomVector(Vector Scale=0.15), Use Orientation Vector=true, Use Rotation Vector=true)
   &CameraOffset(Camera Offset Amount=10)
  ]
  particleUpdateScript [
   &ParticleState()
   &ScaleMeshSize(Scale Factor=&VectorFromFloat(Value=&FloatFromCurve(
      FloatCurve=DataInterface() {propertyValues {Curve {keys [
          {arriveTangent=4.496, interpMode=RCIM_Cubic, leaveTangent=4.496, tangentMode=RCTM_User, time=0, value=1}
          {arriveTangent=0.625, interpMode=RCIM_Cubic, leaveTangent=0.625, time=0.3, value=1.625}
          <at 1 1.625 RCIM_Cubic>
         ]}}}
      Scale Curve=4
     )))
   &ScaleColor(
    Linear Color Curve=DataInterface() {propertyValues {
      AlphaCurve {keys [<at 0 0>,<at 0.165 0.1>,<at 1 0>]}
      BlueCurve {keys [<at 0.0029 0.7396>,<at 0.1845 0.4375>,<at 0.9834 0.0054>]}
      GreenCurve {keys [<at 0.0029 0.2083>,<at 0.1845 0.1689>,<at 0.9834 0.0014>]}
      RedCurve {keys [<at 0.0029 0.1254>,<at 0.1845 0.3579>,<at 0.9834 0.0781>]}
     }}
    Scale Mode=::NewEnumerator2
   )
   &DynamicMaterialParameters(
    Index 0 Param 1=&Add_Float(
     A=&FloatFromCurve(
      FloatCurve=DataInterface() {propertyValues {Curve {keys [{arriveTangent=4.8237, interpMode=RCIM_Cubic, leaveTangent=4.8237, tangentMode=RCTM_User, time=0, value=0}, <at 1 2.25 RCIM_Cubic>]}}}
      Scale Curve=0.5
     )
     B=&RandomRangeFloat(Minimum=0, Maximum=999)
    )
    Module.Index0Param0Write=true
    Module.Index0Param1Write=true
    Module.Index0Param2Write=true
    Module.Index0Param3Write=true
   )
   &PointForce(
    Force Origin=&TransformPosition(Position=<-57600 42904 9988>, Source Space=::World, Destination Space=::Simulation)
    Force Strength=&FloatFromCurve(FloatCurve=DataInterface() {propertyValues {Curve {keys [<at 0 0>,<at 1 -2>]}}}, Scale Curve=100)
    Module.Use Falloff=false
    Module.UseLinearForce=false
   )
   &Drag(Use Linear Drag=true, Use Rotational Drag=true)
   &SolveForcesAndVelocity(Clamp Velocity=false, Limit Acceleration=false)
  ]
  rendererClasses [&NiagaraMeshRendererProperties]
  renderers [{rendererClass=&NiagaraMeshRendererProperties, rendererIndex=0}]
 }
 GlowFlare {
  emitterUpdateScript [
   &EmitterState()
   &SpawnBurst_Instantaneous(Use Spawn Probability=false, UseLoopCountLimit=true)
  ]
  particleSpawnScript [
   &InitializeParticle(
    Color=<rgba 0.012 0 0.02 0.4696>
    Lifetime=0.1
    Sprite Size Mode=::NewEnumerator3
    Uniform Sprite Size=700
    UsePositionOffset=false
    Write Material Random=true
   )
   /SetVariables(Particles.FlareCameraOffset1=22.8579, Particles.FlareCameraOffset2=20) {isSetParameters=true}
  ]
  particleUpdateScript [
   &ParticleState(Kill Particles When Lifetime Has Elapsed=false, Let Infinitely Lived Particles Die When Emitter Deactivates=true)
   /SetVariables(
    Particles.FlareColor1:LinearColor=&MultiplyLinearColor(
     Linear Color=<rgba 0.9844 0.0122 0.1234 0.3478>
     Scale Factor=&MakeFloat4FromVectorAndFloat(
      "Vector (RGB)"=&VectorFromFloat(Value=&RandomRangeFloat(Minimum=0.9, Maximum=1, Evaluation Type=::NewEnumerator1))
      "Float (Alpha)"=1
     )
    )
    Particles.FlareColor2:LinearColor=&MultiplyLinearColor(
     Linear Color=<rgba 0.6042 0.053 0.1614 1>
     Scale Factor=&MakeFloat4FromVectorAndFloat(
      "Vector (RGB)"=&VectorFromFloat(Value=&RandomRangeFloat(Minimum=0.5, Maximum=2, Evaluation Type=::NewEnumerator1))
      "Float (Alpha)"=1
     )
    )
    Particles.FlareSize1:Vector2f=&Vector2DFromFloat(Value=&Multiply_Float(A=1500, B=&RandomRangeFloat(Minimum=0.9, Maximum=1, Evaluation Type=::NewEnumerator1)))
    Particles.FlareSize2:Vector2f=&Vector2DFromFloat(Value=&Multiply_Float(A=1000, B=&RandomRangeFloat(Minimum=0.5, Maximum=1, Evaluation Type=::NewEnumerator1)))
    Particles.GlowColor=&MultiplyLinearColor(
     Linear Color=<rgba 0.2979 0.5819 1 0.2>
     Scale Factor=&MakeFloat4FromVectorAndFloat("Vector (RGB)"=<1 1 1>, "Float (Alpha)"=1)
    )
    Particles.GlowSize=&Vector2DFromFloat(Value=500)
   ) {isSetParameters=true}
   &ScaleColor(ScaleA=true, ScaleRGB=true)
   &FN_Light_Attributes(
    Color=<rgba 0.8637 2.4719 3 1>
    Create Color Attribute=true
    Create Exponent Attribute=true
    Create Radius Attribute=true
    Exponent=4
    Radius=250
    Scale Color=8
   )
   &DynamicMaterialParameters(
    Index 0 Param 1=1
    Module.Index0Float0Write=true
    Module.Index0Float1Write=true
    Module.Index0Float2Write=true
    Module.Index0Float3Write=true
   )
  ]
  rendererClasses [&NiagaraSpriteRendererProperties, &NiagaraLightRendererProperties]
  renderers [
   {rendererClass=&NiagaraSpriteRendererProperties}
   {rendererClass=&NiagaraSpriteRendererProperties, rendererIndex=1}
   {rendererClass=&NiagaraSpriteRendererProperties, rendererIndex=2}
   {rendererClass=&NiagaraSpriteRendererProperties, rendererIndex=3}
   {rendererClass=&NiagaraLightRendererProperties, rendererIndex=4}
  ]
 }
}
summary {
 systemName=NS_Pistol_Canary_MM_Aura_Loop
 emitters [
  {emitterName=Flow, rendererClasses [&NiagaraMeshRendererProperties]}
  {emitterName=GlowFlare, rendererClasses [&NiagaraSpriteRendererProperties, &NiagaraLightRendererProperties]}
 ]
 userVariables [User.bIsLocalPlayer {type=NiagaraBool, defaultValue=true}]
}
